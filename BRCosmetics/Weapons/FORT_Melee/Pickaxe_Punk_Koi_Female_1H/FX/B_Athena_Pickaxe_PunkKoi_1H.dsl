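; Punk Koi one-handed pickaxe Blueprint that animates a fringe flame fade on its idle Niagara effect.
; Left swings, right swings and equipping all restart the FringeFlameFade timeline, after calling parent behavior.
; Each timeline update writes flame opacity to the idle effect's FadeOut parameter.
; The same update sets bIsFrontEnd according to whether the cosmetic is displayed in a frontend context.
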(event EventSwingLeft
  (|CallParentFunction)
  (bind (_direction _flameopacity) (Timeline|FringeFlameFade PlayFromStart :NewTime 0.3)
    (:Update
      (bind _asniagara_particle_system_component (Utilities|Casting|CastToNiagaraComponent (CallFunction|GetIdleVFXComponent)))
      (Effects|Components|ParticleSystem|SetFloatParameter _asniagara_particle_system_component "FadeOut" _flameopacity)
      (if (Reactive|Frontend|IsCosmeticObjectDisplayedinFrontendContext _asniagara_particle_system_component)
        (Effects|Components|ParticleSystem|SetBoolParameter _asniagara_particle_system_component "bIsFrontEnd" true)
        (else
          (Effects|Components|ParticleSystem|SetBoolParameter _asniagara_particle_system_component "bIsFrontEnd" false))))))

(event EventSwingRight
  (|CallParentFunction)
  (Timeline|FringeFlameFade PlayFromStart))

(event FortWeapon|EventOnEquippedAndAttached
  (|CallParentFunction)
  (Utilities|FlowControl|Sequence
    (:then_0
      (Timeline|FringeFlameFade SetNewTime))
    (:then_1
      (Timeline|FringeFlameFade PlayFromStart))))

(fn ConstructionScript
  (|CallParentFunction))
