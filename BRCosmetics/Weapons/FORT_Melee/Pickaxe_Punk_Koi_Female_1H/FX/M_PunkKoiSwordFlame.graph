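# Translucent, unlit, two-sided flame material for Niagara mesh particles on the Punk Koi female one-handed sword pickaxe.
# Panning spooky-face textures, shaped by distorted UV masks, a Niagara-driven top fade and Fresnel, drive opacity and a two-color emissive tinted by time of day.
# Vertex offset combines a clamped bend from Niagara dynamic parameters, sine waves and directional stretch.
# Mobile skips depth fade and the distortion-texture offset.
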Material /BRCosmetics/Weapons/FORT_Melee/Pickaxe_Punk_Koi_Female_1H/FX/M_PunkKoiSwordFlame
graph {
 properties {
  bUsedWithNiagaraMeshParticles=true
  blendMode=BLEND_Translucent
  pixelDepthOffsetMode=PDOM_Legacy
  shadingModel=MSM_Unlit
  twoSided=true
  usedShadingModels=Unlit
 }
 nodes [
  Constant() {r=1}
  n1 = VectorParameter() {group=Color_time_of_day, parameterName=Morning_Color, defaultValue=<rgba 0.2588 0.2588 0.2588 1>}
  n2 = VectorParameter() {group=Color_time_of_day, parameterName=Day_Color, defaultValue=<rgba 1 1 1 1>}
  n3 = VectorParameter() {group=Color_time_of_day, parameterName=Evening_Color, defaultValue=<rgba 0.2154 0.121 0.3686 1>}
  n4 = VectorParameter() {group=Color_time_of_day, parameterName=Night_Color, defaultValue=<rgba 0.2684 0.6583 1 1>}
  n5 = VectorParameter() {parameterName=Color_1, defaultValue=<rgba 0 3 3 1>}
  n6 = VectorParameter() {parameterName=Color_2, defaultValue=<rgba 0 0.7705 1 1>}
  n7 = ScalarParameter() {defaultValue=3, parameterName=ThirdTextureXScale}
  n8 = ScalarParameter() {defaultValue=0.5, parameterName=ThirdTextureYScale}
  n9 = ComponentMask(None=VectorParameter() {group=UVScale, parameterName=UVScale, defaultValue=<rgba 1 1 0 0>}.RGB) {g=true, r=true}
  n10 = Multiply(A=Multiply(A=TextureCoordinate(), B=AppendVector(A=@n7, B=@n8)), B=@n9)
  n11 = Multiply(A=ScalarParameter() {defaultValue=3, parameterName=Speed}, B=Time())
  n12 = MaterialFunctionCall(
   In=TextureSample(UVs=Panner(Coordinate=@n10, Time=@n11) {speedX=0.2, speedY=1}) {texture=/Game/Effects/Fort_Effects/Textures/Tile/T_SpookyFaces_Chroma_Tile}.G
   Contrast=Constant() {r=0.5}
  ) {materialFunction=&CheapContrast}
  n13 = Multiply(
   A=TextureCoordinate()
   B=AppendVector(
    A=ScalarParameter() {defaultValue=4, parameterName=SecondTextureXScale}
    B=ScalarParameter() {defaultValue=1, parameterName=SecondTextureYScale}
   )
  )
  n14 = TextureSample(UVs=Panner(Coordinate=Multiply(A=@n9, B=@n13), Time=@n11) {speedX=-0.5, speedY=1}) {texture=&T_SpookyFaces_Chroma_Tile}
  n15 = Multiply(
   A=ComponentMask(None=VectorParameter() {group=Distortion, parameterName=Distortion UV Scale, sortPriority=3, defaultValue=<rgba 2 1 0 0>}.RGB) {g=true, r=true}
   B=TextureCoordinate() {vTiling=1.5}
  )
  n16 = Multiply(A=Time(), B=ScalarParameter() {defaultValue=3, group=Panner Speed, parameterName=Speed})
  n17 = ComponentMask(None=TextureSample(UVs=Panner(Coordinate=@n15, Time=@n16) {speedX=0.2, speedY=1}) {texture=&T_SpookyFaces_Chroma_Tile}.RGB) {g=true, r=true}
  n18 = TextureCoordinate()
  n19 = Add(A=ComponentMask(None=OneMinus(None=@n18)) {g=true})
  n20 = ComponentMask(None=Add(A=TextureCoordinate(), B=Multiply(A=@n17, B=@n19))) {g=true}
  n21 = SphereMask(A=TextureCoordinate() {vTiling=0.75}, B=Constant2Vector() {r=0.2}, Radius=Constant() {r=0.6}, Hardness=Constant() {r=0.3})
  n22 = Constant() {r=0.5}
  n23 = Constant() {r=0.5}
  n24 = Multiply(
   A=Power(Base=OneMinus(None=@n21)) {constExponent=3}
   B=SphereMask(A=TextureCoordinate() {uTiling=0}, B=Constant2Vector() {g=0.5}, Radius=@n22, Hardness=@n23)
  )
  n25 = TextureCoordinate()
  n26 = Add(
   A=OneMinus(None=ComponentMask(None=@n25) {g=true})
   B=LinearInterpolate(Alpha=DynamicParameter() {defaultValue=<rgba 0 1 1 1>, paramNames [TopFadeOutBlend, Param2, Param3, Param4]}.TopFadeOutBlend) {constA=-0.2, constB=0.7}
  ) {constB=0.5}
  n27 = ScalarParameter() {defaultValue=3, parameterName=TopFadeOutPower}
  n28 = Saturate(None=Subtract(A=Multiply(A=@n20, B=@n24), B=Power(Base=@n26, Exp=@n27) {constExponent=0.5}))
  n29 = SphereMask(A=TextureCoordinate() {vTiling=0}, B=Constant2Vector() {r=0.5}, Radius=@n22, Hardness=@n23)
  n30 = Multiply(A=Add(A=Add(A=@n12.Result, B=@n14.R)) {constB=0.15}, B=Power(Base=Multiply(A=@n28, B=@n29)))
  n31 = Multiply(
   A=Saturate(None=Power(Base=@n30) {constExponent=1.25})
   B=MaterialFunctionCall(Brightness=Constant() {r=1}, Exponent=Constant() {r=1}, Camera Vector=CameraVectorWS(), Vertex / Pixel Normal=VertexNormalWS()) {materialFunction=&MF_Fresnel_FX}.Fresnel Double Sided
  )
  n32 = Saturate(None=@n31)
  MaterialFunctionCall(Emissive=Multiply(
    A=MaterialFunctionCall(Color_Morning=@n1.RGB, Color_Day=@n2.RGB, Color_Evening=@n3.RGB, Color_Night=@n4.RGB) {materialFunction=&Color_Time_of_Day}.Result
    B=LinearInterpolate(A=@n5.RGB, B=@n6.RGB, Alpha=@n32)
   )) {materialFunction=&MF_FX_EmissiveAdjustments}
  n33 = Multiply(A=@n31, B=DepthFade() {fadeDistanceDefault=1})
  Multiply(A=ParticleColor().A, B=ShadingPathSwitch(Default=@n33, Deferred=@n33, Mobile=@n31) {desc=Migrated from FeatureLevelSwitch})
  n34 = DynamicParameter() {desc=Parameters set by Niagara, parameterIndex=1, paramNames [X,Y,Z,WPOIntensity]}
  n35 = ParticlePositionWS()
  n36 = DynamicParameter() {desc=Parameters set by Niagara, parameterIndex=2, defaultValue=<rgba 0 0 20 1>, paramNames [X Offset FE, Y Offset FE, Z Offset FE, Param4]}
  n37 = Transform(None=MaterialFunctionCall(X=@n36.X Offset FE, Y=@n36.X Offset FE, Z=@n36.X Offset FE) {materialFunction=&MakeFloat3}.Result) {transformSourceType=TRANSFORMSOURCE_World, transformType=TRANSFORM_Local}
  n38 = Clamp(
   None=Subtract(
    A=MaterialFunctionCall(X=@n34.X, Y=@n34.X, Z=@n34.X) {materialFunction=&MakeFloat3}.Result
    B=Add(A=@n35, B=@n37)
   )
   Min=ScalarParameter() {defaultValue=-100, group=WPO, parameterName=Clamp_min}
   Max=ScalarParameter() {defaultValue=100, group=WPO, parameterName=Clamp_max}
  ) {maxDefault=1000, minDefault=-1000}
  n39 = OneMinus(None=TextureCoordinate())
  n40 = Multiply(
   A=Power(Base=ComponentMask(None=@n39) {g=true}, Exp=ScalarParameter() {defaultValue=3, group=WPO, parameterName=WPOBendPower})
   B=@n34.WPOIntensity
  )
  n41 = Multiply(A=VertexNormalWS(), B=ScalarParameter() {defaultValue=2, parameterName=WaveStrenght})
  n42 = Constant() {r=1}
  n43 = ScalarParameter() {defaultValue=0.5, parameterName=WaveTilingV}
  n44 = ComponentMask(None=Multiply(A=TextureCoordinate(), B=AppendVector(A=@n42, B=@n43))) {g=true}
  n45 = Multiply(A=Time(), B=ScalarParameter() {defaultValue=2, parameterName=Wavespeed}) {constB=2}
  n46 = Sine(None=Add(A=Power(Base=@n44), B=Add(A=@n44, B=@n45)))
  n47 = VectorParameter() {group=WPO, parameterName=FlameDirection, defaultValue=<rgba 1 1 0 1>}
  n48 = ScalarParameter() {defaultValue=10, group=WPO, parameterName=Distort}
  n49 = TextureCoordinate()
  n50 = ComponentMask(None=@n49) {g=true}
  n51 = Multiply(A=Power(Base=ComponentMask(None=@n49) {r=true}, Exp=Constant() {r=5}), B=Power(Base=OneMinus(None=@n50), Exp=Constant() {r=3}))
  n52 = Add(
   A=Multiply(A=@n40, B=Multiply(A=@n41, B=@n46))
   B=Multiply(A=Transform(None=@n47.RGB) {transformSourceType=TRANSFORMSOURCE_Local}, B=LinearInterpolate(B=@n48, Alpha=@n51))
  )
  n53 = Panner(
   Coordinate=TextureCoordinate() {uTiling=2}
   Time=Multiply(A=Time(), B=ScalarParameter() {defaultValue=1, group=Panner Speed, parameterName=DistortOverlaySpeed})
  ) {speedY=0.8}
  n54 = ComponentMask(None=TextureCoordinate()) {g=true}
  n55 = LinearInterpolate(
   B=LinearInterpolate(
    B=TextureSample(UVs=@n53) {samplerType=SAMPLERTYPE_Normal, texture=&T_Weap_Distort_N_02}.RGB
    Alpha=OneMinus(None=@n54)
   )
   Alpha=ScalarParameter() {defaultValue=1, parameterName=LayeredDistortionStrenght}
  )
  n56 = ShadingPathSwitch(Default=@n55, Deferred=@n55, Mobile=Constant()) {desc=Migrated from FeatureLevelSwitch}
  Reroute(None=Add(A=LinearInterpolate(B=@n38, Alpha=@n40), B=Add(A=@n52, B=@n56)))
 ]
}
