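# Niagara system for the idle flame effect on the Punk Koi Fringe pickaxe, rendered as mesh particles from one Flame emitter.
# Two timed bursts spawn particles that are pinned and oriented each update to sampled bones of a user-supplied skeletal mesh.
# A time-smoothed particle position is stored in an Interpolate Dynamic attribute, capturing motion lag.
# User FadeOut drives a material parameter, while bIsFrontEnd switches position offset and material values for front-end display.
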NiagaraSystem /BRCosmetics/Weapons/FORT_Melee/Pickaxe_Punk_Koi_Female_1H/FX/NS_Punk_Koi_Fringe_IdleFlame
dependencies {usedDataInterfaces [], usedDynamicInputs [], usedModules [&SystemState]}
emitters {Flame {
  emitterUpdateScript [
   &EmitterState(
    Life Cycle Mode=::NewEnumerator1
    Loop Behavior=::NewEnumerator1
    Loop Duration Mode=::NewEnumerator1
    Reset Age On Awaken=true
    Scalability Mode=::NewEnumerator1
    UseLoopDelay=false
   )
   &SpawnBurst_Instantaneous(Use Spawn Probability=false, UseLoopCountLimit=false)
   &SpawnBurst_Instantaneous(Spawn Time=6.7104, Use Spawn Probability=false, UseLoopCountLimit=false) {module=SpawnBurst_Instantaneous001}
  ]
  particleSpawnScript [&InitializeParticle(Color Mode=::NewEnumerator0, Mesh Scale Mode=::NewEnumerator3, UsePositionOffset=false, Write Material Random=true)]
  particleUpdateScript [
   &ParticleState(Kill Particles When Lifetime Has Elapsed=false, Let Infinitely Lived Particles Die When Emitter Deactivates=true)
   &InterpolateOverTime(Rate Of Change=2, Target Vector Value:NiagaraPosition=$Particles.Position, Variable Type=::NewEnumerator2) {module=InterpolateOverTime001}
   /SetVariables(Particles.Interpolate Dynamic=&MakeFloat4FromVectorAndFloat("Vector (RGB)"=$Output.InterpolateOverTime001.Moving Average, "Float (Alpha)"=0.5)) {isSetParameters=true}
   &SolveForcesAndVelocity(Clamp Velocity=false, Limit Acceleration=false)
   &SkeletalMeshLocation(
    Apply Sampled Orientation to Mesh Particles=true
    Bone Sampling Mode=::NewEnumerator4
    Inherit Velocity=false
    Interpolate Mesh Sampling=false
    MaskSpawnGroup=false
    Sampled Position Offset=&Add_Vector(
     A=<0 0 0>
     B=&MakeCustomVectorFromBool(Bool=$User.bIsFrontEnd, True Vector=<-0.5 0 15>, False Vector=<0 0 0>)
    )
    Sampled Position Offset Coordinate Space:ENiagaraMeshTransforms=::NewEnumerator2
    Skeletal Mesh=$User.Skeletal_Mesh
   ) {module=SkeletalMeshLocation001}
   &CameraOffset(Camera Offset Amount=3)
   &DynamicMaterialParameters(
    Index 0 Param 1=$User.FadeOut
    Index 2 Param 1=&MakeCustomFloatFromBool(Bool=$User.bIsFrontEnd, True Float=25, False Float=0)
    Index 2 Param 2=&MakeCustomFloatFromBool(Bool=$User.bIsFrontEnd, True Float=0, False Float=0)
    Index 2 Param 3=&MakeCustomFloatFromBool(Bool=$User.bIsFrontEnd, True Float=-15, False Float=0)
    Module.Param0WriteEnabled=true
    Module.Param0WriteEnabled2=true
    Module.Param1WriteEnabled=true
    Module.Param1WriteEnabled2=true
    Module.Param2WriteEnabled=true
    Module.Param2WriteEnabled2=true
    Module.Param3WriteEnabled=true
    Module.Param3WriteEnabled2=true
    Write Parameter Index 2=true
   )
  ]
  rendererClasses [&NiagaraMeshRendererProperties]
  renderers [{rendererClass=&NiagaraMeshRendererProperties, rendererIndex=0}]
 }}
summary {
 systemName=NS_Punk_Koi_Fringe_IdleFlame
 emitters [{emitterName=Flame, rendererClasses [&NiagaraMeshRendererProperties]}]
 userVariables [
  User.bIsFrontEnd {type=NiagaraBool}
  User.FadeOut {type=NiagaraFloat, defaultValue=0, description=""}
  User.Skeletal_Mesh {
   type=NiagaraDataInterfaceSkeletalMesh
   defaultValue {
    struct=NiagaraExt_VariableValue_DataInterface
    value {dataInterface=&self:NiagaraDataInterfaceSkeletalMesh_1}
   }
  }
 ]
}
