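# Physics asset for the Punk Koi female one-handed axe pickaxe cosmetic, giving physics bodies to its four dynamic flower bones.
# Each flower has an offset capsule and is constrained to a capsule on the Melee_TwoHanded_Handle bone with tight swing and twist limit angles.
# Extra constraints link flowers 1–3 to flower 4 with all angular motion free, so their listed limit angles do not restrict rotation.
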PhysicsAsset /BRCosmetics/Weapons/FORT_Melee/Pickaxe_Punk_Koi_Female_1H/Meshes/Punk_Koi_Female_1h_Axe_Physics
bodies {
 Melee_TwoHanded_Handle [{shapeType=Capsule, length=2, radius=2}]
 dyn_flower_1 [{shapeType=Capsule, length=2, radius=6.605, shapeName=dyn_flower_1_capsule, center=<5 0 0>}]
 dyn_flower_2 [{shapeType=Capsule, length=2, radius=6.605, shapeName=dyn_flower_2_capsule, center=<5 0 0>}]
 dyn_flower_3 [{shapeType=Capsule, length=2, radius=6.605, shapeName=dyn_flower_3_capsule, center=<5 0 0>}]
 dyn_flower_4 [{shapeType=Capsule, length=2, radius=6.605, shapeName=dyn_flower_4_capsule, center=<5 0 0>}]
}
constraints [
 {bone1Name=dyn_flower_1, bone2Name=Melee_TwoHanded_Handle, swing1LimitDegrees=10, swing2LimitDegrees=15, twistLimitDegrees=30}
 {bone1Name=dyn_flower_2, bone2Name=Melee_TwoHanded_Handle, swing1LimitDegrees=10, swing2LimitDegrees=15, twistLimitDegrees=30}
 {bone1Name=dyn_flower_3, bone2Name=Melee_TwoHanded_Handle, swing1LimitDegrees=10, swing2LimitDegrees=15, twistLimitDegrees=30}
 {bone1Name=dyn_flower_4, bone2Name=Melee_TwoHanded_Handle, swing1LimitDegrees=10, swing2LimitDegrees=15, twistLimitDegrees=30}
 {
  bone1Name=dyn_flower_1
  bone2Name=dyn_flower_4
  swing1LimitDegrees=45
  swing1Motion=Free
  swing2LimitDegrees=45
  swing2Motion=Free
  twistLimitDegrees=45
  twistMotion=Free
 }
 {
  bone1Name=dyn_flower_2
  bone2Name=dyn_flower_4
  swing1LimitDegrees=45
  swing1Motion=Free
  swing2LimitDegrees=45
  swing2Motion=Free
  twistLimitDegrees=45
  twistMotion=Free
 }
 {
  bone1Name=dyn_flower_3
  bone2Name=dyn_flower_4
  swing1LimitDegrees=45
  swing1Motion=Free
  swing2LimitDegrees=45
  swing2Motion=Free
  twistLimitDegrees=45
  twistMotion=Free
 }
]
